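# Niagara idle effect for the AnvilAlarmMoth Sparks guitar cosmetic, built from a pulsing heart mesh and a looping spiral of small meshes.
# Spiral particles spawn in one burst, then travel by wrapping time plus index through curves that shape radius, height and scale.
# System-level reactive values, with a fake audio-reactivity loop module, drive waveform pulses in scale, tint and material parameters, split into low/medium/high bands across spiral partitions.
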NiagaraSystem /SparksCosmetics/Instrument/Guitar/AnvilAlarmMoth/Effects/NS_Guitar_AnvilAlarmMoth_IdleFX
dependencies {
 usedDataInterfaces [&NiagaraDataInterfaceCurve, &NiagaraDataInterfaceArrayFloat, &NiagaraDataInterfaceColorCurve]
 usedDynamicInputs [
  &MergeHelper_Vector2Switch
  &MakeIntFromFloat
  /Niagara/DynamicInputs/Lerp/SmoothLerpOverTime_Float
  &RemapRange
  &MakeCustomFloatFromBool
  &SetBoolByFloatComparison
  &FNTimeOfDayScalar
 ]
 usedModules [
  &self:SystemSpawnScript.SystemScriptSource.SystemScriptGraph.NiagaraNodeAssignment_13.SetVariables_#1
  &SystemState
  &self:SystemSpawnScript.SystemScriptSource.SystemScriptGraph.NiagaraNodeAssignment_8.SetVariables_#2
  &Timeline
  /Game/Effects/Niagara/Modules/Audio/FN_FakeAudioReactivityLoop
  &self:SystemSpawnScript.SystemScriptSource.SystemScriptGraph.NiagaraNodeAssignment_12.SetVariables_#3
  &self:SystemSpawnScript.SystemScriptSource.SystemScriptGraph.NiagaraNodeAssignment_2.SetVariables_#4
  &self:SystemSpawnScript.SystemScriptSource.SystemScriptGraph.NiagaraNodeAssignment_6.SetVariables_#5
  &self:SystemSpawnScript.SystemScriptSource.SystemScriptGraph.NiagaraNodeAssignment_7.SetVariables_#6
  &self:SystemSpawnScript.SystemScriptSource.SystemScriptGraph.NiagaraNodeAssignment_10.SetVariables_#7
 ]
}
emitters {
 Heart {
  emitterUpdateScript [
   &EmitterState(Life Cycle Mode=::NewEnumerator1, Loop Behavior=::NewEnumerator1, Loop Duration Mode=::NewEnumerator1, UseLoopDelay=false)
   &SpawnBurst_Instantaneous(Use Spawn Probability=false, UseLoopCountLimit=true)
  ]
  particleSpawnScript [
   &PartitionParticles(ID to Partition:NiagaraInt32=&ReturnExecIndex(), Number Of Partitions=5)
   &InitializeParticle(
    Mesh Scale Mode=::NewEnumerator3
    Mesh Uniform Scale=0.054
    Position Mode=::NewEnumerator1
    Position Offset=<9.8 -5.25 0.5>
    Position Offset Coordinate Space=::Local
    Sprite Size Mode=::NewEnumerator3
    Uniform Sprite Size=9
    UsePositionOffset=true
    Write Material Random=true
   )
   &CameraOffset(Camera Offset Amount=&MergeHelper_FloatSwitch(0=0, 1=1, 2=0, 3=?, Index=$System.ContextSwitch, Num=::NewEnumerator1))
  ]
  particleUpdateScript [
   &ParticleState(Kill Particles When Lifetime Has Elapsed=false, Let Infinitely Lived Particles Die When Emitter Deactivates=true)
   &ScaleMeshSize(Scale Factor=&VectorFromFloat(Value=&Waveform("Waveform [1]"=::NewEnumerator2, "[1] Frequency"=1, "[1] Amplitude Min/Max"=<-0.4 0.3>, "[1] Scale"=$System.reactiveFloat, "[1] Bias"=1))) {module=ScaleMeshSize001}
   &ScaleMeshSize(Scale Factor=&VectorFromFloat(Value=&Lerp_Float(A=$System.FrontEndReactiveScale, B=1, Alpha=$System.Timeline.NormalizedTime)))
   &ScaleColor(
    Scale RGB=&MakeVectorFromLinearColorRGB(Color=&LerpLinearColors(
      StartColor=<rgba 1 1 1 1>
      EndColor=&RandomRangeLinearColor(Minimum=<rgba 1 0.4 0.4 1>, Maximum=<rgba 1 0.4 0.41 1>)
      LerpFactor=&Waveform("Waveform [1]"=::NewEnumerator2, "[1] Frequency"=1, "[1] Amplitude Min/Max"=<0 1.2>, "[1] Scale"=$System.reactiveFloat)
     ))
    ScaleA=true
    ScaleRGB=true
   )
   &DynamicMaterialParameters(
    Index 0 Param 1=$System.EmissiveIntensity
    Index 0 Param 4=&Waveform("Waveform [1]"=::NewEnumerator2, "[1] Frequency"=1, "[1] Amplitude Min/Max"=<-2 2>, "[1] Scale"=$System.reactiveFloat)
    Module.Index0Float0Write=true
    Module.Index0Float1Write=true
    Module.Index0Float2Write=true
    Module.Index0Float3Write=true
   )
  ]
  rendererClasses [&NiagaraMeshRendererProperties]
  renderers [
   {rendererClass=&NiagaraMeshRendererProperties, rendererIndex=0}
   {rendererClass=&NiagaraMeshRendererProperties, rendererIndex=1}
  ]
 }
 Spiral_Particles_TravelingLoop {
  emitterUpdateScript [
   &EmitterState()
   &SpawnBurst_Instantaneous(Spawn Count=50, Use Spawn Probability=false, UseLoopCountLimit=true)
   /SetVariables(Emitter.Height=92, Emitter.SpiralFrequency=1.5, Emitter.SpiralOffset=0.4) {isSetParameters=true}
  ]
  particleSpawnScript [
   &PartitionParticles(Number Of Partitions=4, Partition Mode:ENiagara_PartitionMode_v2=::NewEnumerator5)
   &InitializeParticle(
    Lifetime=5
    Mesh Scale Mode=::NewEnumerator1
    Mesh Uniform Scale Max=0.13
    Mesh Uniform Scale Min=0.09
    Sprite Size Mode=::NewEnumerator1
    Uniform Sprite Size Max=7
    UsePositionOffset=false
    Write Material Random=true
   )
   &InitialMeshOrientation(
    Mesh Orientation Mode=::NewEnumerator2
    Rotation=&RandomRangeVector(Minimum=<0 -0.05 -0.05>, Maximum=<1 0.05 0.05>)
    Use Rotation Vector=true
   )
   /SetVariables(Particles.TravelSpeedParticle:NiagaraFloat=&FloatFromCurve(
     FloatCurve=DataInterface() {propertyValues {Curve {keys [<at 0 1>,<at 0.7931 1>,<at 1 0.85>]}}}
     CurveIndex=&RandomRangeFloat(Minimum=0, Maximum=1)
     Scale Curve=&MergeHelper_FloatSwitch(0=0.2, 1=0.14, 2=0.14, 3=0.14, Index=$Particles.PartitionParticles.Partition, Num=::NewEnumerator2)
    )) {isSetParameters=true}
  ]
  particleUpdateScript [
   &ParticleState(Kill Particles When Lifetime Has Elapsed=false, Let Infinitely Lived Particles Die When Emitter Deactivates=true)
   /SetVariables(
    Particles.CustomIndex:NiagaraFloat=&Frac_Float(Float=&Add_Float(
      A=&ReturnNormalizedExecIndex()
      B=&Add_Float(A=&Multiply_Float(A=$Engine.Time, B=$Particles.TravelSpeedParticle), B=0)
     ))
    Particles.SpiralStrengthWaveform:NiagaraFloat=&Waveform(
     Global Input Position=&ReturnNormalizedExecIndex()
     "Waveform [1]"=::NewEnumerator0
     "[1] Phase Scale"=0.25
     "[1] Frequency"=1
     "[1] Amplitude Min/Max"=<1 1.1>
     "[1] Scale"=19.5
    )
   ) {isSetParameters=true}
   /SetVariables(
    Particles.SpiralFrequency:NiagaraFloat=&FloatFromCurve(
     FloatCurve=DataInterface() {propertyValues {Curve {keys [<at 0 1>,<at 0.6003 1>,<at 0.65 2.15>,<at 1 2.15>]}}}
     CurveIndex=$Particles.CustomIndex
     Scale Curve=$Emitter.SpiralFrequency
    )
    Particles.SpiralStrengthX:NiagaraFloat=&FloatFromCurve(
     FloatCurve=DataInterface() {propertyValues {Curve {keys [
         {arriveTangent=9.4544, interpMode=RCIM_Cubic, leaveTangent=9.4544, tangentMode=RCTM_User, time=0, value=0.1}
         {arriveTangent=2.7134, interpMode=RCIM_Cubic, leaveTangent=2.7134, time=0.0526, value=0.4339}
         {arriveTangent=2.2881, interpMode=RCIM_Cubic, leaveTangent=2.2881, time=0.2002, value=0.6433}
         {arriveTangent=1.0877, interpMode=RCIM_Cubic, leaveTangent=1.0877, time=0.3, value=1}
         {arriveTangent=-2.2818, interpMode=RCIM_Cubic, leaveTangent=-2.2818, time=0.4362, value=0.9}
         {arriveTangent=-2.0831, interpMode=RCIM_Cubic, leaveTangent=-2.0831, time=0.598, value=0.3199}
         {arriveTangent=0.4802, interpMode=RCIM_Cubic, leaveTangent=0.4802, time=0.7656, value=0.2138}
         {arriveTangent=-1.6046, interpMode=RCIM_Cubic, leaveTangent=-1.6046, tangentMode=RCTM_User, time=0.8508, value=0.4413}
         {arriveTangentWeight=0.0624, interpMode=RCIM_Cubic, tangentMode=RCTM_User, time=1.0009, value=0.1981}
        ]}}}
     CurveIndex=$Particles.CustomIndex
     Scale Curve=$Particles.SpiralStrengthWaveform
    )
    Particles.SpiralStrengthY:NiagaraFloat=&FloatFromCurve(
     FloatCurve=DataInterface() {propertyValues {Curve {keys [
         <at 0 0.1389 RCIM_Cubic>
         {arriveTangentWeight=0.0444, interpMode=RCIM_Cubic, leaveTangentWeight=0.0828, tangentMode=RCTM_User, time=0.1247, value=0.2527}
         {arriveTangentWeight=0.0842, interpMode=RCIM_Cubic, leaveTangentWeight=0.0789, tangentMode=RCTM_User, time=0.3571, value=0.2729}
         {arriveTangent=-0.3158, interpMode=RCIM_Cubic, leaveTangent=-0.3158, time=0.5747, value=0.0605}
         {
          arriveTangent=0.0887
          arriveTangentWeight=0.0432
          interpMode=RCIM_Cubic
          leaveTangent=0.0887
          leaveTangentWeight=0.1018
          tangentMode=RCTM_User
          time=0.6895
          value=0.168
         }
         {arriveTangent=-0.4576, interpMode=RCIM_Cubic, leaveTangent=-0.4576, tangentMode=RCTM_User, time=1.0026, value=0.167}
        ]}}}
     CurveIndex=$Particles.CustomIndex
     Scale Curve=$Particles.SpiralStrengthWaveform
    )
   ) {isSetParameters=true}
   /SetVariables(Particles.PositionOffset:NiagaraPosition=&AddVectorToPosition(
     Position=&AddVectorToPosition(
      Position=<0 -1 6>
      Vector=&Multiply_VectorByFloat(
       Vector=&RandomRangeVector(Minimum=<-0.5 -0.5 -1>, Maximum=<0.5 0.5 1>)
       Float=&FloatFromCurve(
        FloatCurve=DataInterface() {propertyValues {Curve {
           defaultValue=1
           keys [
            {arriveTangent=8.5123, interpMode=RCIM_Cubic, leaveTangent=8.5123, tangentMode=RCTM_User, time=0, value=0.5129}
            {arriveTangentWeight=0.3222, interpMode=RCIM_Cubic, leaveTangentWeight=0.0504, tangentMode=RCTM_User, time=0.1129, value=1}
            {arriveTangent=-2.3336, interpMode=RCIM_Cubic, leaveTangent=-2.3336, tangentMode=RCTM_User, time=0.2543, value=0.6719}
            {arriveTangent=-1.2809, interpMode=RCIM_Cubic, leaveTangent=-1.2809, time=0.37, value=0.8763}
            {
             arriveTangent=0.1514
             arriveTangentWeight=0.0226
             interpMode=RCIM_Cubic
             leaveTangent=0.1514
             leaveTangentWeight=0.0234
             tangentMode=RCTM_User
             time=0.5
             value=0.3572
            }
            {
             arriveTangent=0.2217
             arriveTangentWeight=0.0619
             interpMode=RCIM_Cubic
             leaveTangent=0.2217
             leaveTangentWeight=0.0743
             tangentMode=RCTM_User
             time=0.65
             value=0.35
            }
            {
             arriveTangent=0.8408
             arriveTangentWeight=0.0523
             interpMode=RCIM_Cubic
             leaveTangent=0.8408
             leaveTangentWeight=0.3204
             tangentMode=RCTM_User
             time=0.8507
             value=0.52
            }
            {arriveTangent=-0.2719, interpMode=RCIM_Cubic, leaveTangent=-0.2719, tangentMode=RCTM_User, time=1, value=0.35}
           ]
          }}}
        CurveIndex=$Particles.CustomIndex
        Scale Curve=&Waveform(
         Global Input Position=&ReturnNormalizedExecIndex()
         "Waveform [1]"=::NewEnumerator2
         "[1] Phase Scale"=0.1
         "[1] Frequency"=1.1
         "[1] Amplitude Min/Max"=<-1 1>
         "[1] Scale"=&Lerp_Float(A=5, B=7, Alpha=$System.reactiveFloat)
        )
       )
      )
     )
     Vector=&MakeVector(
      X=&Waveform(
       Global Input Position=$Particles.CustomIndex
       Global Phase=&Add_Float(A=&Multiply_Float(A=$Engine.Time, B=$Emitter.SpiralSpeed), B=$Emitter.SpiralOffset)
       "Waveform [1]"=::NewEnumerator0
       "[1] Frequency"=$Particles.SpiralFrequency
       "[1] Amplitude Min/Max"=<-1 1>
       "[1] Scale"=$Particles.SpiralStrengthX
      )
      Y=&Waveform(
       Global Input Position=$Particles.CustomIndex
       Global Phase=&Add_Float(A=&Multiply_Float(A=$Engine.Time, B=$Emitter.SpiralSpeed), B=$Emitter.SpiralOffset)
       "Waveform [1]"=::NewEnumerator1
       "[1] Frequency"=$Particles.SpiralFrequency
       "[1] Amplitude Min/Max"=<-1 1>
       "[1] Scale"=$Particles.SpiralStrengthY
      )
      Z=&Add_Float(
       A=&FloatFromCurve(
        FloatCurve=DataInterface() {propertyValues {Curve {
           defaultValue=0
           keys [
            {tangentMode=RCTM_User, time=0, value=0}
            <at 0.5073 0.3952>
            <at 0.6138 0.4784>
            <at 0.6777 0.7073>
            {tangentMode=RCTM_User, time=1, value=1}
           ]
          }}}
        CurveIndex=$Particles.CustomIndex
        Scale Curve=$Emitter.Height
       )
       B=-25
      )
     )
    )) {isSetParameters=true}
   &ScaleMeshSize(Scale Factor=&Multiply_VectorByFloat(
     Vector=&VectorFromFloat(Value=1)
     Float=&Multiply_Float(
      A=&Multiply_Float(
       A=&MergeHelper_FloatSwitch(
        0=&Waveform(
         "Waveform [1]"=::NewEnumerator2
         "[1] Frequency"=1
         "[1] Amplitude Min/Max"=<-0.25 0.75>
         "[1] Scale"=$System.AudioLow
         "[1] Bias"=&RandomRangeFloat(Minimum=0.8, Maximum=1.1)
        )
        1=&Waveform(
         "Waveform [1]"=::NewEnumerator2
         "[1] Frequency"=1.5
         "[1] Amplitude Min/Max"=<-0.25 0.75>
         "[1] Scale"=$System.AudioMedium
         "[1] Bias"=&RandomRangeFloat(Minimum=0.8, Maximum=1.1)
        )
        2=&Waveform(
         "Waveform [1]"=::NewEnumerator2
         "[1] Frequency"=2
         "[1] Amplitude Min/Max"=<-0.25 0.75>
         "[1] Scale"=$System.AudioHigh
         "[1] Bias"=&RandomRangeFloat(Minimum=0.8, Maximum=1.1)
        )
        3=&Waveform(
         "Waveform [1]"=::NewEnumerator2
         "[1] Frequency"=1
         "[1] Amplitude Min/Max"=<-0.25 0.75>
         "[1] Scale"=$System.AudioLow
         "[1] Bias"=&RandomRangeFloat(Minimum=0.8, Maximum=1.1)
        )
        Index=$Particles.PartitionParticles.Partition
        Num=::NewEnumerator2
       )
       B=1
      )
      B=&FloatFromCurve(
       FloatCurve=DataInterface() {propertyValues {Curve {
          defaultValue=0
          keys [
           {arriveTangent=7.1238, interpMode=RCIM_Cubic, leaveTangent=7.1238, leaveTangentWeight=0.0825, tangentMode=RCTM_User, time=0, value=0}
           {
            arriveTangent=1.3013
            arriveTangentWeight=0.1782
            interpMode=RCIM_Cubic
            leaveTangent=1.3013
            leaveTangentWeight=0.1222
            tangentMode=RCTM_User
            time=0.1323
            value=0.75
           }
           {
            arriveTangent=-4.2095
            arriveTangentWeight=0.0723
            interpMode=RCIM_Cubic
            leaveTangent=-4.2095
            leaveTangentWeight=0.0438
            tangentMode=RCTM_User
            time=0.27
            value=0.3474
           }
           {
            arriveTangent=-0.6475
            arriveTangentWeight=0.1638
            interpMode=RCIM_Cubic
            leaveTangent=-0.6475
            leaveTangentWeight=0.2008
            tangentMode=RCTM_User
            time=0.4082
            value=0.5797
           }
           {
            arriveTangent=-6.3233
            arriveTangentWeight=0.1494
            interpMode=RCIM_Cubic
            leaveTangentWeight=0.0424
            tangentMode=RCTM_Break
            time=0.5601
            value=0
           }
           {
            arriveTangentWeight=0.0424
            interpMode=RCIM_Cubic
            leaveTangent=6.4872
            leaveTangentWeight=0.0531
            tangentMode=RCTM_Break
            time=0.6659
            value=0
           }
           {
            arriveTangent=0.1355
            arriveTangentWeight=0.0769
            interpMode=RCIM_Cubic
            leaveTangent=0.1355
            leaveTangentWeight=0.054
            tangentMode=RCTM_User
            time=0.8622
            value=0.48
           }
           {arriveTangent=-8.1534, interpMode=RCIM_Cubic, leaveTangent=-8.1534, tangentMode=RCTM_User, time=1, value=0}
          ]
         }}}
       CurveIndex=$Particles.CustomIndex
      )
     )
    )) {module=ScaleMeshSize001}
   &ScaleColor(
    Scale RGB=&MakeVectorFromLinearColorRGB(Color=&MergeHelper_ColorSwitch(
      0=&LerpLinearColors(
       StartColor=<rgba 1 1 1 1>
       EndColor=&RandomRangeLinearColor(Minimum=<rgba 1 0.4 0.4 1>, Maximum=<rgba 0.4 1 0.4 1>)
       LerpFactor=&Waveform("Waveform [1]"=::NewEnumerator2, "[1] Frequency"=1, "[1] Amplitude Min/Max"=<0 1.2>, "[1] Scale"=$System.AudioLow)
      )
      1=&LerpLinearColors(
       StartColor=<rgba 1 1 1 1>
       EndColor=&RandomRangeLinearColor(Minimum=<rgba 0.4 1 0.4 1>, Maximum=<rgba 0.4 0.4 1 1>)
       LerpFactor=&Waveform("Waveform [1]"=::NewEnumerator2, "[1] Frequency"=1.5, "[1] Amplitude Min/Max"=<0 1.2>, "[1] Scale"=$System.AudioMedium)
      )
      2=&LerpLinearColors(
       StartColor=<rgba 1 1 1 1>
       EndColor=&RandomRangeLinearColor(Minimum=<rgba 0.4 0.4 1 1>, Maximum=<rgba 1 0.4 0.4 1>)
       LerpFactor=&Waveform("Waveform [1]"=::NewEnumerator2, "[1] Frequency"=2, "[1] Amplitude Min/Max"=<0 1.2>, "[1] Scale"=$System.AudioHigh)
      )
      3=&LerpLinearColors(
       StartColor=<rgba 1 1 1 1>
       EndColor=&RandomRangeLinearColor(Minimum=<rgba 1 0.4 0.4 1>, Maximum=<rgba 0.4 1 0.4 1>)
       LerpFactor=&Waveform("Waveform [1]"=::NewEnumerator2, "[1] Frequency"=1, "[1] Amplitude Min/Max"=<0 1.2>, "[1] Scale"=$System.AudioLow)
      )
      Index=$Particles.PartitionParticles.Partition
      Num=::NewEnumerator2
     ))
    ScaleA=true
    ScaleRGB=true
   ) {module=ScaleColor001}
   &ScaleMeshSize(Scale Factor=&MakeVector(
     X=&Waveform("Waveform [1]"=::NewEnumerator2, "[1] Frequency"=1, "[1] Amplitude Min/Max"=<-0.25 0.25>, "[1] Scale"=$System.AudioLow, "[1] Bias"=1)
     Y=&Waveform("Waveform [1]"=::NewEnumerator2, "[1] Frequency"=2, "[1] Amplitude Min/Max"=<-0.25 0.25>, "[1] Scale"=$System.AudioMedium, "[1] Bias"=1)
     Z=&Waveform("Waveform [1]"=::NewEnumerator2, "[1] Frequency"=3, "[1] Amplitude Min/Max"=<-0.25 0.25>, "[1] Scale"=$System.AudioHigh, "[1] Bias"=1)
    ))
   &ScaleMeshSize(Scale Factor=&VectorFromFloat(Value=&Lerp_Float(
      A=&Lerp_Float(A=$System.FrontEndReactiveScale, B=1, Alpha=$System.Timeline.NormalizedTime)
      B=&FN_FrontEndFloatModifier(Gameplay=1, Front End=1.2)
      Alpha=$System.reactiveFloat
     ))) {module=ScaleMeshSize002}
   &DynamicMaterialParameters(
    Index 0 Param 1=$System.EmissiveIntensity
    Index 0 Param 4=&MergeHelper_FloatSwitch(
     0=&Waveform("Waveform [1]"=::NewEnumerator2, "[1] Frequency"=1, "[1] Amplitude Min/Max"=<0 1>, "[1] Scale"=$System.AudioLow)
     1=&Waveform("Waveform [1]"=::NewEnumerator2, "[1] Frequency"=1.5, "[1] Amplitude Min/Max"=<0 1>, "[1] Scale"=$System.AudioMedium)
     2=&Waveform("Waveform [1]"=::NewEnumerator2, "[1] Frequency"=2, "[1] Amplitude Min/Max"=<0 1>, "[1] Scale"=$System.AudioHigh)
     3=&Waveform("Waveform [1]"=::NewEnumerator2, "[1] Frequency"=1, "[1] Amplitude Min/Max"=<0 1>, "[1] Scale"=$System.AudioLow)
     Index=$Particles.PartitionParticles.Partition
     Num=::NewEnumerator2
    )
    Module.Index0Float0Write=true
    Module.Index0Float1Write=true
    Module.Index0Float2Write=true
    Module.Index0Float3Write=true
   ) {module=DynamicMaterialParameters001}
  ]
  rendererClasses [&NiagaraMeshRendererProperties]
  renderers [{rendererClass=&NiagaraMeshRendererProperties, rendererIndex=0}]
 }
}
summary {
 systemName=NS_Guitar_AnvilAlarmMoth_IdleFX
 emitters [
  {emitterName=Heart, rendererClasses [&NiagaraMeshRendererProperties]}
  {emitterName=Spiral_Particles_TravelingLoop, rendererClasses [&NiagaraMeshRendererProperties]}
 ]
 userVariables [
  User.bIsFrontEnd {type=NiagaraBool}
  User.bIsFrontEndPreview {type=NiagaraBool}
  User.ContextSwitch {type=NiagaraFloat, defaultValue=0, description=""}
 ]
}
