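# Niagara impact effect for the AnvilAlarmMoth guitar pickaxe, with two authored color variants selected by System.VariantInt.
# Core_Flash bursts several spawn groups, switching size, color and curves per group across mesh, sprite and light renderers.
# Sparks and Impact_Smoke001 add streaks and puffs, while FN_PhysMatDebris throws mesh debris tuned by User.PhysicalSurface, with collision disabled.
# Time-of-day curves and front-end modifiers adjust brightness, alpha and spawn radius.
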NiagaraSystem /SparksCosmetics/Instrument/Guitar/AnvilAlarmMoth/Effects/NS_Pickaxe_Guitar_AnvilAlarmMoth_Impact
dependencies {
 usedDataInterfaces [&NiagaraDataInterfaceCurve, &NiagaraDataInterfaceColorCurve, &NiagaraDataInterfaceArrayColor]
 usedDynamicInputs [&MakeIntFromFloat]
 usedModules [
  &self:SystemSpawnScript.SystemScriptSource.SystemScriptGraph.NiagaraNodeAssignment_1.SetVariables_#1
  &self:SystemSpawnScript.SystemScriptSource.SystemScriptGraph.NiagaraNodeAssignment_0.SetVariables_#2
  &SystemState
 ]
}
emitters {
 Core_Flash {
  emitterUpdateScript [
   &EmitterState()
   &SpawnBurst_Instantaneous(Spawn Count=1, Use Spawn Probability=false) {enabled=false}
   &SpawnBurst_Instantaneous(Spawn Count=&RandomRangeInt(Minimum=12, Maximum=13), Spawn Group=1, Use Spawn Probability=false) {module=SpawnBurst_Instantaneous001}
   &SpawnBurst_Instantaneous(Spawn Count=7, Spawn Group=2, Use Spawn Probability=false, UseLoopCountLimit=false) {module=SpawnBurst_Instantaneous002}
   &SpawnBurst_Instantaneous(Spawn Group=3, Use Spawn Probability=false, UseLoopCountLimit=false) {module=SpawnBurst_Instantaneous003}
   &SpawnBurst_Instantaneous(Spawn Group=4, Use Spawn Probability=false, UseLoopCountLimit=false) {module=SpawnBurst_Instantaneous004}
   &SpawnBurst_Instantaneous(Spawn Group=5, Use Spawn Probability=false, UseLoopCountLimit=false) {module=SpawnBurst_Instantaneous005}
   &SpawnBurst_Instantaneous(Spawn Group=6, Use Spawn Probability=false, UseLoopCountLimit=false) {module=SpawnBurst_Instantaneous006}
  ]
  particleSpawnScript [
   /SetVariables(Particles.VisibilityTag=&SpawnGroup()) {isSetParameters=true}
   &InitializeParticle(
    Color=&MergeHelper_ColorSwitch(
     0=&MergeHelper_ColorSwitch(
      0=<rgba 1.5 1.5 1.5 1>
      1=&RandomRangeLinearColor(Minimum=<rgba 1.1022 1.9812 2 1>, Maximum=<rgba 1.1167 1.0827 2 1>)
      2=<rgba 0.1715 0.2709 0.3919 0.8582>
      3=<rgba 1.0106 1.25 0.9326 1>
      4=<rgba 1 1 1 1>
      5=<rgba 0.0515 0.0742 0.0684 0.9>
      6=<rgba 3.3984 4.0391 5 1>
      Index=&SpawnGroup()
      Num=::NewEnumerator5
     )
     1=&MergeHelper_ColorSwitch(
      0=<rgba 1.5 1.5 1.5 1>
      1=&RandomRangeLinearColor(Minimum=<rgba 0 0.088 1 1>, Maximum=<rgba 0.4138 0.0007 1 1>)
      2=&RandomRangeLinearColor(Minimum=<rgba 0.0101 0.0911 0.4844 0.8582>, Maximum=<rgba 0.0259 0 0.612 0.8582>)
      3=<rgba 0.3565 0.1788 1.25 1>
      4=<rgba 1 1 1 1>
      5=<rgba 0.1362 0.1287 0.2031 0.7>
      6=<rgba 2.9653 0 3 1>
      Index=&SpawnGroup()
      Num=::NewEnumerator5
     )
     2=?
     3=?
     Index=$System.VariantInt
     Num=::NewEnumerator0
    )
    Lifetime=&MergeHelper_FloatSwitch(
     0=0.22
     1=&RandomRangeFloat(Minimum=0.15, Maximum=0.2)
     2=0.17
     3=0.18
     4=0.2
     5=0.25
     6=0.2
     Index=&SpawnGroup()
     Num=::NewEnumerator5
    )
    Mesh Scale=&Multiply_VectorByFloat(
     Vector=&Multiply_VectorByFloat(Vector=<1 2 2>, Float=&RandomRangeFloat(Minimum=0.8, Maximum=0.82))
     Float=$System.Global Scale
    )
    Mesh Scale Mode=::NewEnumerator4
    Sprite Size=&Multiply_Vector2D_ByFloat(
     Float=$System.Global Scale
     Vector2D=&MergeHelper_Vector2Switch(
      0=<0 0>
      1=&RandomRangeVector2D(Minimum=<10 55>, Maximum=<13 70>)
      2=&RandomRangeVector2D(Minimum=<30 45>, Maximum=<45 65>)
      3=&Vector2DFromFloat(Value=200)
      4=<1 1>
      5=&Vector2DFromFloat(Value=200)
      6=&Vector2DFromFloat(Value=75)
      Index=$Emitter.SpawnGroup
      Num=::NewEnumerator5
     )
    )
    Sprite Size Mode=::NewEnumerator4
    UsePositionOffset=false
    Write Material Random=true
   )
   &ShapeLocation(
    MaskSpawnGroup=false
    Override Seed=false
    Rotation Angle:NiagaraFloat=&RandomRangeFloat(Minimum=0, Maximum=1)
    Rotation Mode=::NewEnumerator1
    Sphere Radius=&MergeHelper_FloatSwitch(
     0=0
     1=&Multiply_Float(A=0.5, B=$System.Global Scale)
     2=&Multiply_Float(A=2.5, B=$System.Global Scale)
     3=0
     4=0
     5=0
     Index=&SpawnGroup()
     Num=::NewEnumerator4
    )
    UseOwnerScale=true
   )
   &DynamicMaterialParameters(
    Index 0 Param 2=45
    Index 0 Param 3=1
    Index 0 Param 4=1
    Module.Param0WriteEnabled=false
    Module.Param1WriteEnabled=true
    Module.Param2WriteEnabled=true
    Module.Param3WriteEnabled=true
   ) {module=DynamicMaterialParameters001}
   &InitialMeshOrientation(
    Rotation=&RandomRangeVector(Minimum=<-1 -1 -1>, Maximum=<1 1 1>)
    Use Orientation Vector=true
    Use Rotation Vector=true
   )
   &CameraOffset(Camera Offset Amount=&MergeHelper_FloatSwitch(0=20, 1=2, 2=10, 3=10, 4=0, 5=25, 6=20, Index=&SpawnGroup(), Num=::NewEnumerator5))
   &AddVelocityFromPoint(
    Module.Use Falloff Distance=false
    Velocity Strength=&MergeHelper_FloatSwitch(0=0, 1=2, 2=125, 3=0, 4=0, Index=&SpawnGroup(), Num=::NewEnumerator3)
   )
   /SetVariables(Particles.Flare SUBUV:NiagaraFloat=&RandomRangeFloat(Minimum=1, Maximum=3)) {isSetParameters=true}
  ]
  particleUpdateScript [
   &ParticleState()
   &DynamicMaterialParameters(
    Index 0 Param 1=&FloatFromCurve(FloatCurve=DataInterface() {propertyValues {Curve {keys [<at 0.2 0 RCIM_Cubic>, {arriveTangent=3.6451, interpMode=RCIM_Cubic, leaveTangent=3.6451, tangentMode=RCTM_User, time=1, value=1}]}}})
    Module.Index0Float0Write=true
    Module.Index0Float1Write=true
    Module.Index0Float2Write=true
    Module.Index0Float3Write=true
   )
   &ScaleMeshSize(Scale Factor=&VectorFromFloat(Value=&FloatFromCurve(
      FloatCurve=DataInterface() {propertyValues {Curve {
         defaultValue=0.5
         keys [
          {arriveTangent=2.9597, interpMode=RCIM_Cubic, leaveTangent=2.9597, tangentMode=RCTM_User, time=0, value=0}
          {arriveTangent=0.721, interpMode=RCIM_Cubic, leaveTangent=0.721, tangentMode=RCTM_User, time=1, value=1}
         ]
        }}}
      Scale Curve=0.7
     )))
   &ScaleColor(
    Scale Alpha=&MergeHelper_FloatSwitch(
     0=&FloatFromCurve(FloatCurve=DataInterface() {propertyValues {Curve {keys [<at 0 1>,<at 1 0>]}}})
     1=&FloatFromCurve(FloatCurve=DataInterface() {propertyValues {Curve {
         defaultValue=1
         keys [
          {arriveTangent=-0.592, interpMode=RCIM_Cubic, leaveTangent=-0.592, tangentMode=RCTM_User, time=0, value=1}
          {interpMode=RCIM_Cubic, tangentMode=RCTM_User, time=1, value=0}
         ]
        }}})
     2=&FloatFromCurve(FloatCurve=DataInterface() {propertyValues {Curve {
         defaultValue=1
         keys [
          {arriveTangent=-0.5163, interpMode=RCIM_Cubic, leaveTangent=-0.5163, tangentMode=RCTM_User, time=0, value=1}
          {interpMode=RCIM_Cubic, tangentMode=RCTM_User, time=1, value=0}
         ]
        }}})
     3=&FloatFromCurve(
      FloatCurve=DataInterface() {propertyValues {Curve {
         defaultValue=1
         keys [
          {interpMode=RCIM_Cubic, tangentMode=RCTM_User, time=0, value=1}
          {arriveTangent=-1.2989, interpMode=RCIM_Cubic, leaveTangent=-1.2989, tangentMode=RCTM_User, time=1, value=0}
         ]
        }}}
      Scale Curve=&FN_FrontEndFloatModifier(
       Gameplay=&FloatFromCurve(
        FloatCurve=DataInterface() {propertyValues {Curve {keys [
            <at 0 0.65 RCIM_Cubic>
            {arriveTangent=0.875, interpMode=RCIM_Cubic, leaveTangent=0.875, time=0.1, value=0.7}
            {arriveTangentWeight=0.1329, interpMode=RCIM_Cubic, leaveTangentWeight=0.0759, tangentMode=RCTM_User, time=0.4, value=1}
            {arriveTangentWeight=0.0801, interpMode=RCIM_Cubic, leaveTangentWeight=0.1002, tangentMode=RCTM_User, time=0.6, value=1}
            {arriveTangent=-0.8778, interpMode=RCIM_Cubic, leaveTangent=-0.8778, time=0.9, value=0.7}
            <at 0.9987 0.65 RCIM_Cubic>
           ]}}}
        CurveIndex=Linked() {linkedVariable {name=NPC.FortniteNPC.TimeOfDay_0-1, type=NiagaraFloat}}
       )
       Front End=0.8
      )
     )
     4=1
     5=&FloatFromCurve(
      FloatCurve=DataInterface() {propertyValues {Curve {
         defaultValue=1
         keys [
          {interpMode=RCIM_Cubic, tangentMode=RCTM_User, time=0, value=1}
          {arriveTangent=-2.1802, interpMode=RCIM_Cubic, leaveTangent=-2.1802, tangentMode=RCTM_User, time=1, value=0}
         ]
        }}}
      Scale Curve=&FloatFromCurve(
       FloatCurve=DataInterface() {propertyValues {Curve {
          defaultValue=1
          keys [
           {arriveTangent=0.664, interpMode=RCIM_Cubic, leaveTangent=0.664, tangentMode=RCTM_User, time=0, value=0.45}
           {interpMode=RCIM_Cubic, tangentMode=RCTM_User, time=0.5, value=1}
           {tangentMode=RCTM_User, time=1, value=0.45}
          ]
         }}}
       CurveIndex=Linked() {linkedVariable {name=NPC.FortniteNPC.TimeOfDay_0-1, type=NiagaraFloat}}
       Scale Curve=1.2
      )
     )
     6=&FloatFromCurve(FloatCurve=DataInterface() {propertyValues {Curve {keys [
          <at 0.0045 1 RCIM_Cubic>
          {arriveTangent=-2.0573, arriveTangentWeight=0.5281, interpMode=RCIM_Cubic, leaveTangent=-2.0573, tangentMode=RCTM_User, time=1, value=0}
         ]}}})
     Index=&SpawnGroup()
     Num=::NewEnumerator5
    )
    Scale RGB=&MergeHelper_VectorSwitch(
     0=&MergeHelper_VectorSwitch(
      0=<1 1 1>
      1=&Multiply_VectorByFloat(
       Vector=&MakeVectorFromLinearColorRGB(Color=&ColorFromCurve(ColorCurve=DataInterface() {propertyValues {
           AlphaCurve {keys []}
           BlueCurve {keys [<at 0 1.9586 RCIM_Cubic>,<at 0.4313 1.5>,<at 0.9403 1>]}
           GreenCurve {keys [<at 0 3 RCIM_Cubic>,<at 0.4313 1.177>,<at 0.9403 0.6875>]}
           RedCurve {keys [<at 0 2.6812 RCIM_Cubic>,<at 0.4313 0.8638>,<at 0.9403 0.8266>]}
          }}))
       Float=&FloatFromCurve(
        FloatCurve=DataInterface() {propertyValues {Curve {keys [
            <at 0.1 0.8 RCIM_Cubic>
            {arriveTangentWeight=0.1329, interpMode=RCIM_Cubic, leaveTangentWeight=0.0759, tangentMode=RCTM_User, time=0.45, value=1}
            {arriveTangentWeight=0.0801, interpMode=RCIM_Cubic, leaveTangentWeight=0.1002, tangentMode=RCTM_User, time=0.55, value=1}
            <at 0.9 0.8 RCIM_Cubic>
           ]}}}
        CurveIndex=Linked() {linkedVariable {name=NPC.FortniteNPC.TimeOfDay_0-1, type=NiagaraFloat}}
       )
      )
      2=&VectorFromFloat(Value=&FN_FrontEndFloatModifier(Gameplay=1, Front End=0.7))
      3=&Multiply_VectorByFloat(
       Vector=&MakeVectorFromLinearColorRGB(Color=&ColorFromCurve(ColorCurve=DataInterface() {propertyValues {
           AlphaCurve {keys []}
           BlueCurve {keys [<at 0 0.5391 RCIM_Cubic>,<at 0.6157 0.8932>,<at 1 0.6556>]}
           GreenCurve {keys [<at 0 0.9643 RCIM_Cubic>,<at 0.6157 0.985>,<at 1 0.5171>]}
           RedCurve {keys [<at 0 1 RCIM_Cubic>,<at 0.6157 0.6824>,<at 1 0.4627>]}
          }}))
       Float=&FloatFromCurve(
        FloatCurve=DataInterface() {propertyValues {Curve {keys [
            <at 0.1 0.75 RCIM_Cubic>
            {arriveTangentWeight=0.1329, interpMode=RCIM_Cubic, leaveTangentWeight=0.0759, tangentMode=RCTM_User, time=0.45, value=1}
            {arriveTangentWeight=0.0801, interpMode=RCIM_Cubic, leaveTangentWeight=0.1002, tangentMode=RCTM_User, time=0.55, value=1}
            <at 0.9 0.75 RCIM_Cubic>
           ]}}}
        CurveIndex=Linked() {linkedVariable {name=NPC.FortniteNPC.TimeOfDay_0-1, type=NiagaraFloat}}
        Scale Curve=&FloatFromCurve(FloatCurve=DataInterface() {propertyValues {Curve {keys [{arriveTangent=-7.8135, interpMode=RCIM_Cubic, leaveTangent=-7.8135, tangentMode=RCTM_User, time=0.1356, value=4}, <at 1 1 RCIM_Cubic>]}}})
       )
      )
      4=<1 1 1>
      5=&MakeVectorFromLinearColorRGB(Color=&ColorFromCurve(ColorCurve=DataInterface() {propertyValues {
          AlphaCurve {keys []}
          BlueCurve {keys [<at 0.0724 0.9635 RCIM_Cubic>,<at 0.9556 0.6556>]}
          GreenCurve {keys [<at 0.0724 0.5044 RCIM_Cubic>,<at 0.9556 0.4719>]}
          RedCurve {keys [<at 0.0724 0.3162 RCIM_Cubic>,<at 0.9556 0.1933>]}
         }}))
      6=&Multiply_VectorByFloat(
       Vector=&MakeVectorFromLinearColorRGB(Color=&ColorFromCurve(ColorCurve=DataInterface() {propertyValues {
           AlphaCurve {keys []}
           BlueCurve {keys [<at 0.0037 1>,<at 0.1816 0.8696>,<at 0.7686 0.9695>]}
           GreenCurve {keys [<at 0.0037 1>,<at 0.1816 1>,<at 0.7686 0.5712>]}
           RedCurve {keys [<at 0.0037 1>,<at 0.1816 0.5833>,<at 0.7686 0.404>]}
          }}))
       Float=&FloatFromCurve(
        FloatCurve=DataInterface() {propertyValues {Curve {keys [<at 0 6>,<at 0.4 2>,<at 1 1>]}}}
        Scale Curve=&FloatFromCurve(
         FloatCurve=DataInterface() {propertyValues {Curve {keys [
             <at 0.1 0.8 RCIM_Cubic>
             {arriveTangentWeight=0.1329, interpMode=RCIM_Cubic, leaveTangentWeight=0.0759, tangentMode=RCTM_User, time=0.35, value=1}
             {arriveTangentWeight=0.0801, interpMode=RCIM_Cubic, leaveTangentWeight=0.1002, tangentMode=RCTM_User, time=0.65, value=1}
             <at 0.9 0.8 RCIM_Cubic>
            ]}}}
         CurveIndex=Linked() {linkedVariable {name=NPC.FortniteNPC.TimeOfDay_0-1, type=NiagaraFloat}}
        )
       )
      )
      Index=&SpawnGroup()
      Num=::NewEnumerator5
     )
     1=&MergeHelper_VectorSwitch(
      0=<1 1 1>
      1=&Multiply_VectorByFloat(
       Vector=&MakeVectorFromLinearColorRGB(Color=&ColorFromCurve(ColorCurve=DataInterface() {propertyValues {
           AlphaCurve {keys []}
           BlueCurve {keys [<at 0.0661 3 RCIM_Cubic>,<at 0.3959 2>,<at 0.9945 0.6771>]}
           GreenCurve {keys [<at 0.0661 0 RCIM_Cubic>,<at 0.3959 0>,<at 0.9945 0.0208>]}
           RedCurve {keys [<at 0.0661 0.6494 RCIM_Cubic>,<at 0.3959 0.3879>,<at 0.9945 0>]}
          }}))
       Float=&FloatFromCurve(
        FloatCurve=DataInterface() {propertyValues {Curve {keys [
            <at 0.1 0.8 RCIM_Cubic>
            {arriveTangentWeight=0.1329, interpMode=RCIM_Cubic, leaveTangentWeight=0.0759, tangentMode=RCTM_User, time=0.45, value=1}
            {arriveTangentWeight=0.0801, interpMode=RCIM_Cubic, leaveTangentWeight=0.1002, tangentMode=RCTM_User, time=0.55, value=1}
            <at 0.9 0.8 RCIM_Cubic>
           ]}}}
        CurveIndex=Linked() {linkedVariable {name=NPC.FortniteNPC.TimeOfDay_0-1, type=NiagaraFloat}}
       )
      )
      2=&VectorFromFloat(Value=&FN_FrontEndFloatModifier(Gameplay=1, Front End=0.7))
      3=&Multiply_VectorByFloat(
       Vector=&MakeVectorFromLinearColorRGB(Color=&ColorFromCurve(ColorCurve=DataInterface() {propertyValues {
           AlphaCurve {keys []}
           BlueCurve {keys [<at 0 2 RCIM_Cubic>,<at 0.462 1>,<at 1 0.6556>]}
           GreenCurve {keys [<at 0 0.1198 RCIM_Cubic>,<at 0.462 0.0495>,<at 1 0.3164>]}
           RedCurve {keys [<at 0 0.5939 RCIM_Cubic>,<at 0.462 0.3852>,<at 1 0>]}
          }}))
       Float=&FloatFromCurve(
        FloatCurve=DataInterface() {propertyValues {Curve {keys [
            <at 0.1 0.75 RCIM_Cubic>
            {arriveTangentWeight=0.1329, interpMode=RCIM_Cubic, leaveTangentWeight=0.0759, tangentMode=RCTM_User, time=0.45, value=1}
            {arriveTangentWeight=0.0801, interpMode=RCIM_Cubic, leaveTangentWeight=0.1002, tangentMode=RCTM_User, time=0.55, value=1}
            <at 0.9 0.75 RCIM_Cubic>
           ]}}}
        CurveIndex=Linked() {linkedVariable {name=NPC.FortniteNPC.TimeOfDay_0-1, type=NiagaraFloat}}
        Scale Curve=&FloatFromCurve(FloatCurve=DataInterface() {propertyValues {Curve {keys [{arriveTangent=-6.5809, interpMode=RCIM_Cubic, leaveTangent=-6.5809, tangentMode=RCTM_User, time=0.14, value=5}, <at 1 1 RCIM_Cubic>]}}})
       )
      )
      4=<1 1 1>
      5=&MakeVectorFromLinearColorRGB(Color=&ColorFromCurve(ColorCurve=DataInterface() {propertyValues {
          AlphaCurve {keys []}
          BlueCurve {keys [<at 0 1 RCIM_Cubic>,<at 0.7475 0.849>]}
          GreenCurve {keys [<at 0 0 RCIM_Cubic>,<at 0.7475 0.0579>]}
          RedCurve {keys [<at 0 0.4407 RCIM_Cubic>,<at 0.7475 0.0088>]}
         }}))
      6=&Multiply_VectorByFloat(
       Vector=&MakeVectorFromLinearColorRGB(Color=&ColorFromCurve(ColorCurve=DataInterface() {propertyValues {
           AlphaCurve {keys []}
           BlueCurve {keys [<at 0.0037 1>,<at 0.8239 0.9695>]}
           GreenCurve {keys [<at 0.0037 0>,<at 0.8239 0.0996>]}
           RedCurve {keys [<at 0.0037 0.8155>,<at 0.8239 0.0769>]}
          }}))
       Float=&FloatFromCurve(
        FloatCurve=DataInterface() {propertyValues {Curve {keys [<at 0 4>,<at 0.4 1.75>,<at 1 1>]}}}
        Scale Curve=&FloatFromCurve(
         FloatCurve=DataInterface() {propertyValues {Curve {keys [
             <at 0.1 0.8 RCIM_Cubic>
             {arriveTangentWeight=0.1329, interpMode=RCIM_Cubic, leaveTangentWeight=0.0759, tangentMode=RCTM_User, time=0.35, value=1}
             {arriveTangentWeight=0.0801, interpMode=RCIM_Cubic, leaveTangentWeight=0.1002, tangentMode=RCTM_User, time=0.65, value=1}
             <at 0.9 0.8 RCIM_Cubic>
            ]}}}
         CurveIndex=Linked() {linkedVariable {name=NPC.FortniteNPC.TimeOfDay_0-1, type=NiagaraFloat}}
        )
       )
      )
      Index=&SpawnGroup()
      Num=::NewEnumerator5
     )
     2=?
     3=?
     Index=$System.VariantInt
     Num=::NewEnumerator0
    )
    ScaleA=true
    ScaleRGB=true
   )
   &ScaleSpriteSize(
    Scale Factor=&MergeHelper_Vector2Switch(
     0=<0 0>
     1=&MakeVector2D(
      X=&FloatFromCurve(FloatCurve=DataInterface() {propertyValues {Curve {
          defaultValue=1
          keys [
           {interpMode=RCIM_Cubic, tangentMode=RCTM_User, time=0, value=0}
           {arriveTangentWeight=0.2236, interpMode=RCIM_Cubic, leaveTangentWeight=0.1491, tangentMode=RCTM_User, time=0.7, value=1}
           {arriveTangent=-2.2754, interpMode=RCIM_Cubic, leaveTangent=-2.2754, tangentMode=RCTM_User, time=1, value=0.5}
          ]
         }}})
      Y=&FloatFromCurve(FloatCurve=DataInterface() {propertyValues {Curve {
          defaultValue=1
          keys [
           {arriveTangent=3.143, interpMode=RCIM_Cubic, leaveTangent=3.143, tangentMode=RCTM_User, time=0, value=0}
           {arriveTangent=0.8537, interpMode=RCIM_Cubic, leaveTangent=0.8537, tangentMode=RCTM_User, time=1, value=1}
          ]
         }}})
     )
     2=&Vector2DFromFloat(Value=&FloatFromCurve(FloatCurve=DataInterface() {propertyValues {Curve {
          defaultValue=1
          keys [
           {arriveTangent=0.9682, interpMode=RCIM_Cubic, leaveTangent=0.9682, tangentMode=RCTM_User, time=0, value=0}
           {interpMode=RCIM_Cubic, tangentMode=RCTM_User, time=1, value=1}
          ]
         }}}))
     3=&Multiply_Vector2D_ByFloat(
      Float=&FloatFromCurve(FloatCurve=DataInterface() {propertyValues {Curve {
          defaultValue=0
          keys [
           {interpMode=RCIM_Cubic, tangentMode=RCTM_User, time=0, value=1}
           {arriveTangent=-1.4808, interpMode=RCIM_Cubic, leaveTangent=-1.4808, tangentMode=RCTM_User, time=1, value=0}
          ]
         }}})
      Vector2D=<1 1>
     )
     4=<0 0>
     5=&Vector2DFromFloat(Value=&FloatFromCurve(FloatCurve=DataInterface() {propertyValues {Curve {
          defaultValue=0
          keys [
           {arriveTangent=2.4894, interpMode=RCIM_Cubic, leaveTangent=2.4894, tangentMode=RCTM_User, time=0, value=0}
           {interpMode=RCIM_Cubic, tangentMode=RCTM_User, time=1, value=1}
          ]
         }}}))
     6=&Multiply_Vector2D_ByFloat(
      Float=&FloatFromCurve(FloatCurve=DataInterface() {propertyValues {Curve {
          defaultValue=1
          keys [
           {arriveTangent=1.8312, interpMode=RCIM_Cubic, leaveTangent=1.8312, tangentMode=RCTM_User, time=0, value=0.2}
           {arriveTangent=0.2628, arriveTangentWeight=0.4048, interpMode=RCIM_Cubic, leaveTangent=0.2628, tangentMode=RCTM_User, time=1, value=1}
          ]
         }}})
      Vector2D=<1 1>
     )
     Index=&SpawnGroup()
     Num=::NewEnumerator5
    )
    Scale Sprite Size Mode=::NewEnumerator2
   )
   &SolveForcesAndVelocity(Clamp Velocity=false, Limit Acceleration=false)
   /SetVariables(Particles.Flare Dyn 1:Vector4f=&MakeVector4(
     X=&FloatFromCurve(FloatCurve=DataInterface() {propertyValues {Curve {keys [<at 0 1 RCIM_Cubic>,<at 0.8 0 RCIM_Cubic>]}}})
     Y=2
     Z=1
     W=1
    )) {isSetParameters=true}
   &FN_Light_Attributes(
    Color=&MergeHelper_ColorSwitch(0=<rgba 0.954 1 0.9505 1>, 1=<rgba 0.3515 0.1269 1 1>, 2=?, 3=?, Index=$System.VariantInt, Num=::NewEnumerator0)
    Create Color Attribute=true
    Create Exponent Attribute=true
    Create Position Attribute=true
    Create Radius Attribute=true
    Exponent=5
    Position=<10 0 0>
    Radius=90
    Scale Color=&FloatFromCurve(
     FloatCurve=DataInterface() {propertyValues {Curve {
        defaultValue=1
        keys [{interpMode=RCIM_Cubic, tangentMode=RCTM_User, time=0, value=1}, {interpMode=RCIM_Cubic, tangentMode=RCTM_User, time=1, value=0}]
       }}}
     Scale Curve=10
    )
   )
  ]
  rendererClasses [
   &NiagaraMeshRendererProperties
   &NiagaraSpriteRendererProperties
   &NiagaraLightRendererProperties
  ]
  renderers [
   {rendererClass=&NiagaraMeshRendererProperties, rendererIndex=0}
   {rendererClass=&NiagaraSpriteRendererProperties, rendererIndex=1}
   {rendererClass=&NiagaraSpriteRendererProperties, rendererIndex=2}
   {rendererClass=&NiagaraSpriteRendererProperties, rendererIndex=3}
   {rendererClass=&NiagaraLightRendererProperties, rendererIndex=4}
   {rendererClass=&NiagaraSpriteRendererProperties, rendererIndex=5}
   {rendererClass=&NiagaraSpriteRendererProperties, rendererIndex=6}
  ]
 }
 FN_PhysMatDebris {
  emitterUpdateScript [
   &EmitterState()
   &SpawnBurst_Instantaneous(
    Spawn Count=&RandomRangeInt(Minimum=5, Maximum=10)
    Use Spawn Probability=false
    UseLoopCountLimit=false
   )
  ]
  particleSpawnScript [
   /SetVariables(Particles.BiasedRandom=&FloatFromCurve(
     FloatCurve=DataInterface() {propertyValues {Curve {keys [
         <at 0.2 1 RCIM_Cubic>
         {arriveTangent=-3.3394, interpMode=RCIM_Cubic, leaveTangent=-3.3394, tangentMode=RCTM_User, time=0.5, value=0.5}
         <at 0.8 0 RCIM_Cubic>
        ]}}}
     CurveIndex=&RandomRangeFloat(Minimum=0, Maximum=1)
    )) {isSetParameters=true}
   &InitializeParticle(
    Color Mode=::NewEnumerator0
    Lifetime=&RandomRangeFloat(Minimum=0.5, Maximum=1.25)
    Mass=&Lerp_Float(A=0.6, B=1.2, Alpha=$Particles.BiasedRandom)
    Mass Mode=::NewEnumerator1
    Mesh Scale=&Multiply_VectorByFloat(
     Vector=&RandomRangeVector(Minimum=<0.9 0.4 0.5>, Maximum=<1 1 1>)
     Float=&Lerp_Float(A=0.7, B=1, Alpha=$Particles.BiasedRandom)
    )
    Mesh Scale Mode=::NewEnumerator4
    UsePositionOffset=false
    Write Material Random=true
   )
   &FN_PhysicalSurface(
    Dirt Mesh Count=2
    Glass Mesh Count=3
    Ice Mesh Count=2
    Metal Gravity=&RandomRangeFloat(Minimum=-350, Maximum=-1800)
    Metal Mesh Count=2
    Metal Velocity=&RandomRangeFloat(Minimum=80, Maximum=375)
    Metal Velocity Cone Angle=130
    Physical Surface=$User.PhysicalSurface
    Stone Gravity=&RandomRangeFloat(Minimum=-700, Maximum=-1250)
    Stone Mesh Count=2
    Stone Velocity=&RandomRangeFloat(Minimum=80, Maximum=550)
    Stone Velocity Cone Angle=120
    Wood Gravity=&RandomRangeFloat(Minimum=-450, Maximum=-900)
    Wood Mesh Count=2
    Wood Scale=1.5
    Wood Velocity=&RandomRangeFloat(Minimum=80, Maximum=600)
    Wood Velocity Cone Angle=130
   )
   &AddVelocityInCone(
    Cone Angle=$Output.FN_PhysicalSurface.SurfaceResponseConeAngle
    Module.Use Velocity Falloff On Cone Axis=false
    Velocity Strength=$Output.FN_PhysicalSurface.SurfaceResponseVelocity
   )
   &InitialMeshOrientation(Use Rotation Vector=false)
   &AddRotationalVelocity(Rotation Rate=&RandomVector(Vector Scale=&RandomRangeFloat(Minimum=1, Maximum=2)))
   &ApplyInitialForces()
   /SetVariables(Particles.GravityOverrideValue=$Output.FN_PhysicalSurface.GravityOverride) {isSetParameters=true}
  ]
  particleUpdateScript [
   &ParticleState()
   &GravityForce(Gravity=$Particles.GravityOverrideValue)
   &Collision(
    Advanced Aging Rate=1
    Age Colliding Particles=false
    CPU Collision Trace Channel=ECC_WorldStatic
    Collision Enabled=&Boolean_AndOperation(
     Bool1=&SetBoolByFloatComparison(A=$Particles.Initial.Lifetime, B=1)
     Bool2=&SetBoolByFloatComparison(A=$Particles.Age, B=0.12)
    )
    Control Roll On Collision=true
    Friction=1
    Friction During a Bounce=0.25
    Max CPU Trace Length=100
    Module.Control Rest State Via Surface Normal=false
    Module.EnableMaxCollisionCount=false
    Module.Randomize Collision Normal Vector=true
    Randomize Collision Normal=0.25
    ScaleRateOfRotation=false
   ) {enabled=false}
   &SolveForcesAndVelocity(Clamp Velocity=false, Limit Acceleration=false)
   &ScaleMeshSize(Scale Factor=&VectorFromFloat(Value=&FloatFromCurve(FloatCurve=DataInterface() {propertyValues {Curve {keys [<at 0.6 1 RCIM_Cubic>, {arriveTangent=-6.308, interpMode=RCIM_Cubic, leaveTangent=-6.308, tangentMode=RCTM_User, time=1, value=0}]}}})))
   &DynamicMaterialParameters(Module.Param0WriteEnabled=true, Module.Param1WriteEnabled=true, Module.Param2WriteEnabled=false, Module.Param3WriteEnabled=false)
  ]
  rendererClasses [&NiagaraSpriteRendererProperties, &NiagaraMeshRendererProperties]
  renderers [
   {rendererClass=&NiagaraSpriteRendererProperties}
   {rendererClass=&NiagaraMeshRendererProperties, rendererIndex=1}
   {rendererClass=&NiagaraMeshRendererProperties, rendererIndex=2}
   {rendererClass=&NiagaraMeshRendererProperties, rendererIndex=3}
   {rendererClass=&NiagaraMeshRendererProperties, rendererIndex=4}
   {rendererClass=&NiagaraMeshRendererProperties, rendererIndex=5}
   {rendererClass=&NiagaraMeshRendererProperties, rendererIndex=6}
   {rendererClass=&NiagaraMeshRendererProperties, rendererIndex=7}
   {rendererClass=&NiagaraMeshRendererProperties, rendererIndex=8}
  ]
 }
 Impact_Smoke001 {
  emitterUpdateScript [
   &EmitterState()
   &SpawnBurst_Instantaneous(
    Spawn Count=&RandomRangeInt(Minimum=2, Maximum=3)
    Use Spawn Probability=false
    UseLoopCountLimit=false
   )
  ]
  particleSpawnScript [
   &InitializeParticle(
    Color=&MergeHelper_ColorSwitch(
     0=&RandomRangeLinearColor(Minimum=<rgba 0.6195 0.8421 1 0.1>, Maximum=<rgba 0.6255 0.56 1 0.1>)
     1=&RandomRangeLinearColor(Minimum=<rgba 0.7747 0.9065 1 0.1>, Maximum=<rgba 0.8053 0.9192 1 0.1>)
     2=?
     3=?
     Index=$System.VariantInt
     Num=::NewEnumerator0
    )
    Lifetime Max=0.5
    Lifetime Min=0.4
    Lifetime Mode=::NewEnumerator1
    Sprite Rotation Mode=::NewEnumerator2
    Sprite Size Mode=::NewEnumerator3
    Uniform Sprite Size=&Multiply_Float(A=$System.Global Scale, B=&RandomRangeFloat(Minimum=115, Maximum=130))
    UsePositionOffset=false
    Write Material Random=true
   )
   &CameraOffset(Camera Offset Amount=&RandomRangeFloat(Minimum=40, Maximum=50, Evaluation Type=::NewEnumerator1))
   &ShapeLocation(
    Hemisphere Distribution=<1 0.5>
    MaskSpawnGroup=false
    Override Seed=false
    Rotation Angle=90
    Rotation Mode=::NewEnumerator1
    Sphere Radius=&MakeCustomFloatFromBool(Bool=$User.bIsFrontEnd, True Float=5, False Float=10)
    UseOwnerScale=true
   )
   &AddVelocity(Velocity=&Multiply_VectorByFloat(Vector=$Particles.ShapeLocation.ShapeVector, Float=10))
  ]
  particleUpdateScript [
   &ParticleState()
   &ScaleSpriteSize(Uniform Curve Sprite Scale=DataInterface() {propertyValues {Curve {keys [
        {arriveTangent=0.5339, interpMode=RCIM_Cubic, leaveTangent=0.5339, tangentMode=RCTM_User, time=0, value=0}
        {arriveTangent=0.5865, interpMode=RCIM_Cubic, leaveTangent=0.5865, tangentMode=RCTM_User, time=0.3904, value=0.7232}
        {arriveTangent=0.4342, interpMode=RCIM_Cubic, leaveTangent=0.4342, tangentMode=RCTM_User, time=1, value=1}
       ]}}})
   &ScaleColor(
    Scale Alpha=&FloatFromCurve(FloatCurve=DataInterface() {propertyValues {Curve {keys [{interpMode=RCIM_Cubic, leaveTangentWeight=0.2194, tangentMode=RCTM_User, time=0, value=0}, <at 0.25 0.5 RCIM_Cubic>, <at 1 0 RCIM_Cubic>]}}})
    ScaleA=true
    ScaleRGB=true
   )
   &DynamicMaterialParameters(
    Index 0 Param 1=&Add_Float(
     A=&RandomRangeFloat(Minimum=0, Maximum=1)
     B=&FloatFromCurve(
      FloatCurve=DataInterface() {propertyValues {Curve {keys [
          <at 0 0 RCIM_Cubic>
          {arriveTangent=1.5947, interpMode=RCIM_Cubic, leaveTangent=1.5947, tangentMode=RCTM_User, time=0.2557, value=0.5257}
          <at 1 1 RCIM_Cubic>
         ]}}}
      CurveIndex=$Particles.Age
      Scale Curve=0.25
     )
    )
    Module.Param0WriteEnabled=true
    Module.Param1WriteEnabled=true
    Module.Param2WriteEnabled=true
    Module.Param3WriteEnabled=true
   ) {module=DynamicMaterialParameters001}
   &Drag(Use Linear Drag=true, Use Rotational Drag=true)
   &SolveForcesAndVelocity(Clamp Velocity=false, Limit Acceleration=false)
  ]
  rendererClasses [&NiagaraSpriteRendererProperties]
  renderers [
   {rendererClass=&NiagaraSpriteRendererProperties}
   {rendererClass=&NiagaraSpriteRendererProperties, rendererIndex=1}
  ]
 }
 Sparks {
  emitterUpdateScript [
   &EmitterState()
   &SpawnBurst_Instantaneous(Spawn Count=25, Use Spawn Probability=false, UseLoopCountLimit=false)
  ]
  particleSpawnScript [
   &InitializeParticle(
    Color=&MergeHelper_ColorSwitch(
     0=&SelectLinearColorFromArray(
      Array Sampling Mode=::NewEnumerator0
      Direct Array Index=?
      Color Selection Array=DataInterface() {propertyValues {ColorData [<rgba 0.6452 1.8083 2 1>, <rgba 0.6588 0.6482 2 1>]}}
      Random Seed=0
     )
     1=&SelectLinearColorFromArray(
      Array Sampling Mode=::NewEnumerator0
      Direct Array Index=?
      Color Selection Array=DataInterface() {propertyValues {ColorData [<rgba 0 0.2283 3 1>, <rgba 1.206 0 2 1>]}}
      Random Seed=0
     )
     2=?
     3=?
     Index=$System.VariantInt
     Num=::NewEnumerator0
    )
    Lifetime Max=0.45
    Lifetime Min=0.1
    Lifetime Mode=::NewEnumerator1
    Sprite Size Mode=::NewEnumerator3
    Uniform Sprite Size=&Multiply_Float(A=$System.Global Scale, B=&RandomRangeFloat(Minimum=2, Maximum=4))
    UsePositionOffset=false
    Write Material Random=true
   )
   &ShapeLocation(
    MaskSpawnGroup=false
    Override Seed=false
    Sphere Radius=&Multiply_Float(A=2, B=&FN_FrontEndFloatModifier(Gameplay=0, Front End=0.6))
    UseOwnerScale=true
   )
   &CameraOffset(Camera Offset Amount=25)
   &AddVelocity(
    Module.Use Falloff Distance=false
    Override Seed=false
    Velocity Mode=::NewEnumerator1
    Velocity Speed=&Multiply_Float(A=&RandomRangeFloat(Minimum=400, Maximum=750), B=$System.Global Scale)
   )
  ]
  particleUpdateScript [
   &ParticleState()
   &AccelerationForce(Acceleration=&RandomRangeVector(Minimum=<0 0 -500>, Maximum=<0 0 -500>))
   &Drag(Drag=&RandomRangeFloat(Minimum=5, Maximum=9), Use Linear Drag=true, Use Rotational Drag=true)
   &SolveForcesAndVelocity(Clamp Velocity=false, Limit Acceleration=false)
   &ScaleSpriteSizeBySpeed(Max Scale Factor=<0.4 6>, Min Scale Factor=<0.5 1>)
   &ScaleSpriteSize(
    Uniform Curve Scale=&Multiply_Float(A=1, B=1)
    Uniform Curve Sprite Scale=DataInterface() {propertyValues {Curve {keys [{interpMode=RCIM_Cubic, leaveTangentWeight=0.7932, tangentMode=RCTM_User, time=0.65, value=1}, <at 1 0 RCIM_Cubic>]}}}
   )
   &ScaleColor(
    Scale Alpha=&FloatFromCurve(FloatCurve=DataInterface() {propertyValues {Curve {
        defaultValue=1
        keys [{interpMode=RCIM_Cubic, tangentMode=RCTM_User, time=0.5, value=1}, {interpMode=RCIM_Cubic, tangentMode=RCTM_User, time=1, value=0}]
       }}})
    Scale RGB=&VectorFromFloat(Value=&FloatFromCurve(
      FloatCurve=DataInterface() {propertyValues {Curve {keys [
          {interpMode=RCIM_Cubic, tangentMode=RCTM_User, time=0, value=4}
          <at 0.5 1.5>
          {interpMode=RCIM_Cubic, tangentMode=RCTM_User, time=1, value=1}
         ]}}}
      Scale Curve=&FN_FrontEndFloatModifier(Gameplay=4, Front End=2)
     ))
    ScaleA=true
    ScaleRGB=true
   )
   /SetVariables(
    Particles.GlowColor=&MakeLinearColorFromVectorAndFloat(
     "Vector (RGB)"=&Multiply_Vector(A=&MakeVectorFromLinearColorRGB(Color=$Particles.Color), B=<0.5 0.5 0.5>)
     "Float (Alpha)"=&FloatFromCurve(
      FloatCurve=DataInterface() {propertyValues {Curve {keys [<at 0.5 1>,<at 1 0>]}}}
      Scale Curve=&FN_FrontEndFloatModifier(Gameplay=0.08, Front End=0.03)
     )
    )
    Particles.GlowSize=&Multiply_Vector2D(A=$Particles.SpriteSize, B=<3 2>)
   ) {isSetParameters=true}
  ]
  rendererClasses [&NiagaraSpriteRendererProperties]
  renderers [
   {rendererClass=&NiagaraSpriteRendererProperties}
   {rendererClass=&NiagaraSpriteRendererProperties, rendererIndex=1}
  ]
 }
}
summary {
 systemName=NS_Pickaxe_Guitar_AnvilAlarmMoth_Impact
 emitters [
  {
   emitterName=FN_PhysMatDebris
   rendererClasses [&NiagaraSpriteRendererProperties, &NiagaraMeshRendererProperties]
  }
  {
   emitterName=Core_Flash
   rendererClasses [
    &NiagaraMeshRendererProperties
    &NiagaraSpriteRendererProperties
    &NiagaraLightRendererProperties
   ]
  }
  {emitterName=Impact_Smoke001, rendererClasses [&NiagaraSpriteRendererProperties]}
  {emitterName=Sparks, rendererClasses [&NiagaraSpriteRendererProperties]}
 ]
 userVariables [
  User.bIsFrontEnd {type=NiagaraBool}
  User.bIsFrontEndPreview {type=NiagaraBool}
  User.PhysicalSurface {type=NiagaraInt32}
  User.VariantID {type=NiagaraFloat, defaultValue=0, description=""}
 ]
}
