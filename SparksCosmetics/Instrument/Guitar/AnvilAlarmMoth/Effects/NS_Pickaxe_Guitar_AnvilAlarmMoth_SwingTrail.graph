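# Niagara swing trail for the AnvilAlarmMoth guitar pickaxe cosmetic, built from four short-lived ribbon emitters.
# Shared anim-trail modules spawn and update base, main and two accent ribbons from skeletal bone or socket indices.
# VariantID drives a switch flag in each accent emitter, while VariantInt selects colors for the base and main trails.
# Time-of-day and front-end modifiers scale opacity; dynamic material parameters animate main and accent ribbons.
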NiagaraSystem /SparksCosmetics/Instrument/Guitar/AnvilAlarmMoth/Effects/NS_Pickaxe_Guitar_AnvilAlarmMoth_SwingTrail
dependencies {
 usedDataInterfaces [&NiagaraDataInterfaceColorCurve, &NiagaraDataInterfaceCurve, &NiagaraDataInterfaceSocketReader]
 usedDynamicInputs [&MakeIntFromFloat, &MergeHelper_FloatSwitch]
 usedModules [
  &self:SystemSpawnScript.SystemScriptSource.SystemScriptGraph.NiagaraNodeAssignment_2.SetVariables_#1
  &self:SystemSpawnScript.SystemScriptSource.SystemScriptGraph.NiagaraNodeAssignment_3.SetVariables_#2
  &self:SystemSpawnScript.SystemScriptSource.SystemScriptGraph.NiagaraNodeAssignment_4.SetVariables_#3
  &SystemState
  &self:SystemSpawnScript.SystemScriptSource.SystemScriptGraph.NiagaraNodeAssignment_1.SetVariables_#4
  &self:SystemSpawnScript.SystemScriptSource.SystemScriptGraph.NiagaraNodeAssignment_0.SetVariables_#5
  &self:SystemSpawnScript.SystemScriptSource.SystemScriptGraph.NiagaraNodeAssignment_5.SetVariables_#6
  &self:SystemSpawnScript.SystemScriptSource.SystemScriptGraph.NiagaraNodeAssignment_6.SetVariables_#7
 ]
}
emitters {
 Accent_trail {
  emitterUpdateScript [
   &EmitterState()
   &FN_SpawnAnimTrail()
   /SetVariables(Emitter.VariantSwitch:NiagaraBool=&SetBoolByFloatComparison(A=$User.VariantID, B=1, Comparison Type=::NewEnumerator6)) {isSetParameters=true}
  ]
  particleSpawnScript [
   &FN_InitializeAnimTrailBoneIndecies(First Bone or Socket Index=1) {module=FN_InitializeAnimTrailBoneIndecies001}
   &FN_InitializeAnimTrail(
    AnimTrailWidth=7
    Color=&MultiplyLinearColor(Linear Color=$System.TOD Color, Scale Factor=<1 1 1 1>)
    Lifetime=0.15
   ) {module=FN_InitializeAnimTrail001}
   &AddVelocity(Velocity=<0 500 0>) {enabled=false}
  ]
  particleUpdateScript [
   &ScaleRibbonWidth(Ribbon Width Scale=&FloatFromCurve(FloatCurve=DataInterface() {propertyValues {Curve {keys [
         {arriveTangent=7.6644, interpMode=RCIM_Cubic, leaveTangent=7.6644, tangentMode=RCTM_User, time=0, value=0.5}
         {arriveTangentWeight=0.0745, interpMode=RCIM_Cubic, leaveTangentWeight=0.2981, tangentMode=RCTM_User, time=0.2, value=1}
         {arriveTangent=-2.1703, arriveTangentWeight=0.3708, interpMode=RCIM_Cubic, leaveTangent=-2.1703, tangentMode=RCTM_User, time=1, value=0}
        ]}}}))
   &FN_UpdateAnimTrail(Lock Particles=true) {module=FN_UpdateAnimTrail001}
   &ScaleColor(
    Scale Alpha=&FloatFromCurve(
     FloatCurve=DataInterface() {propertyValues {Curve {keys [
         {arriveTangent=9.9264, interpMode=RCIM_Cubic, leaveTangent=9.9264, tangentMode=RCTM_User, time=0, value=0.3}
         {interpMode=RCIM_Cubic, leaveTangent=-0.962, tangentMode=RCTM_Break, time=0.1, value=0.8}
         {arriveTangent=-0.5809, interpMode=RCIM_Cubic, leaveTangent=-0.5809, tangentMode=RCTM_User, time=0.5, value=0.15}
         {arriveTangent=-0.2139, interpMode=RCIM_Cubic, tangentMode=RCTM_Break, time=1, value=0}
        ]}}}
     Scale Curve=&FloatFromCurve(
      FloatCurve=DataInterface() {propertyValues {Curve {keys [
          <at 0 0.88 RCIM_Cubic>
          {arriveTangentWeight=0.0858, interpMode=RCIM_Cubic, leaveTangentWeight=0.0558, tangentMode=RCTM_User, time=0.4434, value=1}
          {arriveTangentWeight=0.0572, interpMode=RCIM_Cubic, leaveTangentWeight=0.1462, tangentMode=RCTM_User, time=0.6, value=1}
          <at 1 0.88 RCIM_Cubic>
         ]}}}
      CurveIndex=Linked() {linkedVariable {name=NPC.FortniteNPC.TimeOfDay_0-1, type=NiagaraFloat}}
      Scale Curve=$System.AlphaFade
     )
    )
    ScaleA=true
    ScaleRGB=true
   )
   &SolveForcesAndVelocity(Clamp Velocity=false, Limit Acceleration=false)
   &DynamicMaterialParameters(
    Index 0 Param 1=0.5
    Index 1 Param 3=&FloatFromCurve(
     FloatCurve=DataInterface() {propertyValues {Curve {defaultValue=0, keys [{tangentMode=RCTM_User, time=0, value=1}, {tangentMode=RCTM_User, time=1, value=0}]}}}
     CurveIndex=&Frac_Float(Float=&Multiply_Float(A=$Engine.Time, B=1))
    )
    Index 1 Param 4=1
    Index 2 Param 1=5
    Index 2 Param 2=2
    Module.Index0Float0Write=true
    Module.Index0Float1Write=true
    Module.Index0Float2Write=true
    Module.Index0Float3Write=true
    Module.Index1Float0Write=true
    Module.Index1Float1Write=true
    Module.Index1Float2Write=true
    Module.Index1Float3Write=true
    Module.Index2Float0Write=true
    Module.Index2Float1Write=true
    Module.Index2Float2Write=true
    Module.Index2Float3Write=true
    Write Parameter Index 1=::NewEnumerator1
    Write Parameter Index 2=::NewEnumerator1
   )
  ]
  rendererClasses [&NiagaraRibbonRendererProperties]
  renderers [{rendererClass=&NiagaraRibbonRendererProperties}]
 }
 Accent_trail_Variant {
  emitterUpdateScript [
   &EmitterState()
   &FN_SpawnAnimTrail()
   /SetVariables(Emitter.VariantSwitch:NiagaraBool=&SetBoolByFloatComparison(A=$User.VariantID, B=0)) {isSetParameters=true}
  ]
  particleSpawnScript [
   &FN_InitializeAnimTrailBoneIndecies(First Bone or Socket Index=1) {module=FN_InitializeAnimTrailBoneIndecies001}
   &FN_InitializeAnimTrail(
    AnimTrailWidth=7
    Color=&MultiplyLinearColor(Linear Color=$System.TOD Color, Scale Factor=<1 1 1 1>)
    Lifetime=0.15
   ) {module=FN_InitializeAnimTrail001}
   &AddVelocity(Velocity=<0 500 0>) {enabled=false}
  ]
  particleUpdateScript [
   &ScaleRibbonWidth(Ribbon Width Scale=&FloatFromCurve(FloatCurve=DataInterface() {propertyValues {Curve {keys [
         {arriveTangent=7.6644, interpMode=RCIM_Cubic, leaveTangent=7.6644, tangentMode=RCTM_User, time=0, value=0.5}
         {arriveTangentWeight=0.0745, interpMode=RCIM_Cubic, leaveTangentWeight=0.2981, tangentMode=RCTM_User, time=0.2, value=1}
         {arriveTangent=-2.1703, arriveTangentWeight=0.3708, interpMode=RCIM_Cubic, leaveTangent=-2.1703, tangentMode=RCTM_User, time=1, value=0}
        ]}}}))
   &FN_UpdateAnimTrail(Lock Particles=true) {module=FN_UpdateAnimTrail001}
   &ScaleColor(
    Scale Alpha=&FloatFromCurve(
     FloatCurve=DataInterface() {propertyValues {Curve {keys [
         {arriveTangent=9.9264, interpMode=RCIM_Cubic, leaveTangent=9.9264, tangentMode=RCTM_User, time=0, value=0.3}
         {interpMode=RCIM_Cubic, leaveTangent=-0.962, tangentMode=RCTM_Break, time=0.1, value=0.8}
         {arriveTangent=-0.5809, interpMode=RCIM_Cubic, leaveTangent=-0.5809, tangentMode=RCTM_User, time=0.5, value=0.15}
         {arriveTangent=-0.2139, interpMode=RCIM_Cubic, tangentMode=RCTM_Break, time=1, value=0}
        ]}}}
     Scale Curve=&FloatFromCurve(
      FloatCurve=DataInterface() {propertyValues {Curve {keys [
          <at 0 0.88 RCIM_Cubic>
          {arriveTangentWeight=0.0858, interpMode=RCIM_Cubic, leaveTangentWeight=0.0558, tangentMode=RCTM_User, time=0.4434, value=1}
          {arriveTangentWeight=0.0572, interpMode=RCIM_Cubic, leaveTangentWeight=0.1462, tangentMode=RCTM_User, time=0.6, value=1}
          <at 1 0.88 RCIM_Cubic>
         ]}}}
      CurveIndex=Linked() {linkedVariable {name=NPC.FortniteNPC.TimeOfDay_0-1, type=NiagaraFloat}}
      Scale Curve=$System.AlphaFade
     )
    )
    ScaleA=true
    ScaleRGB=true
   )
   &SolveForcesAndVelocity(Clamp Velocity=false, Limit Acceleration=false)
   &DynamicMaterialParameters(
    Index 0 Param 1=0.5
    Index 1 Param 3=&FloatFromCurve(
     FloatCurve=DataInterface() {propertyValues {Curve {defaultValue=0, keys [{tangentMode=RCTM_User, time=0, value=1}, {tangentMode=RCTM_User, time=1, value=0}]}}}
     CurveIndex=&Frac_Float(Float=&Multiply_Float(A=$Engine.Time, B=1))
    )
    Index 1 Param 4=1
    Index 2 Param 1=5
    Index 2 Param 2=2
    Module.Index0Float0Write=true
    Module.Index0Float1Write=true
    Module.Index0Float2Write=true
    Module.Index0Float3Write=true
    Module.Index1Float0Write=true
    Module.Index1Float1Write=true
    Module.Index1Float2Write=true
    Module.Index1Float3Write=true
    Module.Index2Float0Write=true
    Module.Index2Float1Write=true
    Module.Index2Float2Write=true
    Module.Index2Float3Write=true
    Write Parameter Index 1=::NewEnumerator1
    Write Parameter Index 2=::NewEnumerator1
   )
  ]
  rendererClasses [&NiagaraRibbonRendererProperties]
  renderers [{rendererClass=&NiagaraRibbonRendererProperties}]
 }
 NE_Main_Trail {
  emitterSpawnScript [&FN_RibbonIDOffsetUtility()]
  emitterUpdateScript [
   &EmitterState()
   &FN_SpawnAnimTrail(AnimTrailWidthMode=::FromSecond)
   /SetVariables(Emitter.Curve for Colors=DataInterface() {propertyValues {
      bExposeCurve=true
      AlphaCurve {keys [<at 0 0>,<at 0.1513 0.0905>,<at 0.3099 0.5>,<at 0.499 0.1>,<at 0.6504 0.8>,<at 0.7808 0.2>,<at 0.8838 1>,<at 0.9981 0>]}
      BlueCurve {keys [<at 0.0491 0.25>,<at 0.1344 0.0339>,<at 0.1739 0.0339>,<at 0.2334 0.375>,<at 0.326 0.7223>,<at 0.4 0.0339>,<at 0.4371 0.1181>,<at 0.4813 0.2188>,<at 0.5425 0.3116>,<at 0.5691 0.0339>,<at 0.6021 0.375>,<at 0.6544 0.4815>,<at 0.8097 0.2188>,<at 0.85 0.25>,<at 0.8943 0.1821>,<at 0.9361 0.1178>,<at 0.9909 0.0339>]}
      GreenCurve {keys [<at 0.0491 0.1979>,<at 0.1344 0.0268>,<at 0.1739 0.0268>,<at 0.2334 0.0986>,<at 0.326 0.4732>,<at 0.4 0.0268>,<at 0.4371 0.0572>,<at 0.4813 0.165>,<at 0.5425 0.2243>,<at 0.5691 0.0268>,<at 0.6021 0.0986>,<at 0.6544 0.3328>,<at 0.8097 0.165>,<at 0.85 0.1979>,<at 0.8943 0.0837>,<at 0.9361 0.0746>,<at 0.9909 0.0268>]}
      RedCurve {keys [<at 0.0491 0.1875>,<at 0.1344 0.0254>,<at 0.1739 0.0254>,<at 0.2334 0.1959>,<at 0.326 0.3736>,<at 0.4 0.0254>,<at 0.4371 0.041>,<at 0.4813 0.1502>,<at 0.5425 0.1956>,<at 0.5691 0.0254>,<at 0.6021 0.1959>,<at 0.6544 0.2787>,<at 0.8097 0.1502>,<at 0.85 0.1875>,<at 0.8943 0.064>,<at 0.9361 0.066>,<at 0.9909 0.0254>]}
     }}) {enabled=false, isSetParameters=true}
   /SetVariables(Emitter.Curve for Colors=DataInterface() {propertyValues {
      bExposeCurve=true
      AlphaCurve {keys [<at 0 0>,<at 0.1513 0.0905>,<at 0.3099 0.5>,<at 0.499 0.1>,<at 0.6504 0.8>,<at 0.7808 0.2>,<at 0.8838 1>,<at 0.9981 0>]}
      BlueCurve {keys [<at 0.0491 0.25>,<at 0.1344 0.0339>,<at 0.1739 0.0339>,<at 0.2334 0.375>,<at 0.326 0.6631>,<at 0.4 0.0339>,<at 0.4371 0.1181>,<at 0.4813 0.2188>,<at 0.5425 0.3116>,<at 0.5691 0.0339>,<at 0.6021 0.375>,<at 0.6544 0.444>,<at 0.8097 0.2188>,<at 0.85 0.25>,<at 0.8943 0.1821>,<at 0.9361 0.1178>,<at 0.9909 0.0339>]}
      GreenCurve {keys [<at 0.0491 0.1979>,<at 0.1344 0.0268>,<at 0.1739 0.0268>,<at 0.2334 0.0986>,<at 0.326 0.4343>,<at 0.4 0.0268>,<at 0.4371 0.0572>,<at 0.4813 0.165>,<at 0.5425 0.2243>,<at 0.5691 0.0268>,<at 0.6021 0.0986>,<at 0.6544 0.3069>,<at 0.8097 0.165>,<at 0.85 0.1979>,<at 0.8943 0.0837>,<at 0.9361 0.0746>,<at 0.9909 0.0268>]}
      RedCurve {keys [<at 0.0491 0.1875>,<at 0.1344 0.0254>,<at 0.1739 0.0254>,<at 0.2334 0.1959>,<at 0.326 0.343>,<at 0.4 0.0254>,<at 0.4371 0.041>,<at 0.4813 0.1502>,<at 0.5425 0.1956>,<at 0.5691 0.0254>,<at 0.6021 0.1959>,<at 0.6544 0.257>,<at 0.8097 0.1502>,<at 0.85 0.1875>,<at 0.8943 0.064>,<at 0.9361 0.066>,<at 0.9909 0.0254>]}
     }}) {isSetParameters=true}
   /SetVariables(Emitter.Curve for Colors=DataInterface() {propertyValues {
      bExposeCurve=true
      AlphaCurve {keys [<at 0 0>,<at 0.1513 0.0905>,<at 0.3099 0.5>,<at 0.499 0.1>,<at 0.6504 0.8>,<at 0.7808 0.2>,<at 0.8838 1>,<at 0.9981 0>]}
      BlueCurve {keys [<at 0.0491 0.25>,<at 0.1344 0.0339>,<at 0.1739 0.0339>,<at 0.326 0.7223>,<at 0.4 0.0339>,<at 0.4813 0.2188>,<at 0.5425 0.3116>,<at 0.5691 0.0339>,<at 0.6021 0.4021>,<at 0.6544 0.4815>,<at 0.8097 0.2188>,<at 0.85 0.25>,<at 0.9909 0.0339>]}
      GreenCurve {keys [<at 0.0491 0.1979>,<at 0.1344 0.0268>,<at 0.1739 0.0268>,<at 0.326 0.4732>,<at 0.4 0.0268>,<at 0.4813 0.165>,<at 0.5425 0.2243>,<at 0.5691 0.0268>,<at 0.6021 0.2821>,<at 0.6544 0.3328>,<at 0.8097 0.165>,<at 0.85 0.1979>,<at 0.9909 0.0268>]}
      RedCurve {keys [<at 0.0491 0.1875>,<at 0.1344 0.0254>,<at 0.1739 0.0254>,<at 0.326 0.3736>,<at 0.4 0.0254>,<at 0.4813 0.1502>,<at 0.5425 0.1956>,<at 0.5691 0.0254>,<at 0.6021 0.2399>,<at 0.6544 0.2787>,<at 0.8097 0.1502>,<at 0.85 0.1875>,<at 0.9909 0.0254>]}
     }}) {enabled=false, isSetParameters=true}
  ]
  particleSpawnScript [
   &FN_InitializeAnimTrailBoneIndecies()
   &FN_InitializeAnimTrail(Color=$System.TOD Color, Lifetime=0.13)
  ]
  particleUpdateScript [
   &SolveForcesAndVelocity(Clamp Velocity=false, Limit Acceleration=false)
   &FN_UpdateAnimTrail(Lock Particles=true) {module=FN_UpdateAnimTrail001}
   &DynamicMaterialParameters(
    Index 0 Param 1=0.3
    Index 1 Param 3=&FloatFromCurve(
     FloatCurve=DataInterface() {propertyValues {Curve {keys [<at 0 0>, {arriveTangent=1.9814, tangentMode=RCTM_User, time=1, value=1}]}}}
     CurveIndex=&ModuloFloat(A=$Emitter.Age, B=1)
     Scale Curve=0
    )
    Index 1 Param 4=1
    Index 2 Param 1=10
    Index 2 Param 2=1
    Index 2 Param 3=0.5
    Index 2 Param 4=0.6
    Module.Index0Float0Write=true
    Module.Index0Float1Write=true
    Module.Index0Float2Write=true
    Module.Index0Float3Write=true
    Module.Index1Float0Write=true
    Module.Index1Float1Write=true
    Module.Index1Float2Write=true
    Module.Index1Float3Write=true
    Module.Index2Float0Write=true
    Module.Index2Float1Write=true
    Module.Index2Float2Write=true
    Module.Index2Float3Write=true
    Write Parameter Index 1=::NewEnumerator1
    Write Parameter Index 2=::NewEnumerator1
   )
   &ScaleColor(
    Scale Alpha=&FloatFromCurve(
     FloatCurve=DataInterface() {propertyValues {Curve {keys [{interpMode=RCIM_Cubic, tangentMode=RCTM_User, time=0, value=1}, {arriveTangent=-5.7105, tangentMode=RCTM_User, time=0.35, value=0}]}}}
     Scale Curve=&Multiply_Float(
      A=$System.AlphaFade
      B=&FloatFromCurve(
       FloatCurve=DataInterface() {propertyValues {Curve {keys [
           <at 0.1 0.8 RCIM_Cubic>
           {arriveTangentWeight=0.0858, interpMode=RCIM_Cubic, leaveTangentWeight=0.0558, tangentMode=RCTM_User, time=0.4434, value=1}
           {arriveTangentWeight=0.0572, interpMode=RCIM_Cubic, leaveTangentWeight=0.1462, tangentMode=RCTM_User, time=0.6, value=1}
           <at 0.9 0.8 RCIM_Cubic>
          ]}}}
       CurveIndex=Linked() {linkedVariable {name=NPC.FortniteNPC.TimeOfDay_0-1, type=NiagaraFloat}}
       Scale Curve=&MakeCustomFloatFromBool(Bool=$User.bIsFrontEndPreview, True Float=0.6, False Float=0.8)
      )
     )
    )
    Scale RGB=&MergeHelper_VectorSwitch(
     0=&MakeVectorFromLinearColorRGB(Color=<rgba 1 1 1 1>)
     1=&MakeVectorFromLinearColorRGB(Color=<rgba 0.5202 0.5396 1 1>)
     2=?
     3=?
     Index=$System.VariantInt
     Num=::NewEnumerator0
    )
    ScaleA=true
    ScaleRGB=true
   )
   &ScaleRibbonWidth(Ribbon Width Scale=&FloatFromCurve(FloatCurve=DataInterface() {propertyValues {Curve {keys [<at 0 1>,<at 1 0>]}}}))
  ]
  rendererClasses [&NiagaraRibbonRendererProperties]
  renderers [{rendererClass=&NiagaraRibbonRendererProperties}]
 }
 Trail {
  emitterSpawnScript [&FN_RibbonIDOffsetUtility()]
  emitterUpdateScript [
   &EmitterState()
   &FN_SpawnAnimTrail(AnimTrailWidthMode=::FromFirst)
  ]
  particleSpawnScript [
   &FN_InitializeAnimTrailBoneIndecies()
   &FN_InitializeAnimTrail(Color=<rgba 1 1 1 0.05>, Lifetime=0.18) {module=FN_InitializeAnimTrail001}
   &AddVelocity(Velocity=<0 500 0>) {enabled=false}
  ]
  particleUpdateScript [
   &SolveForcesAndVelocity(Clamp Velocity=false, Limit Acceleration=false)
   &FN_UpdateAnimTrail()
   &DynamicMaterialParameters(
    Index 0 Param 1=1
    Index 0 Param 3=0.1
    Module.Index0Param0Write=true
    Module.Index0Param1Write=true
    Module.Index0Param2Write=true
    Module.Index0Param3Write=true
    Module.Index1Param0Write=true
    Module.Index1Param1Write=true
    Module.Index1Param2Write=true
    Module.Index1Param3Write=true
    Write Parameter Index 1=true
   ) {enabled=false}
   /SetVariables(Particles.TopLineColor:LinearColor=&MakeLinearColorFromVectorAndFloat(
     "Vector (RGB)"=&Multiply_Vector(
      A=&MakeVectorFromLinearColorRGB(Color=$System.TOD Color)
      B=&MakeVectorFromLinearColorRGB(Color=&ColorFromCurve(ColorCurve=DataInterface() {propertyValues {
          AlphaCurve {keys [<at 0 1 RCIM_Cubic>]}
          BlueCurve {keys [<at 0 1 RCIM_Cubic>,<at 0.5861 0.5664 RCIM_Cubic>]}
          GreenCurve {keys [<at 0 0.4964 RCIM_Cubic>,<at 0.5861 0.2146 RCIM_Cubic>]}
          RedCurve {keys [<at 0 0.2455 RCIM_Cubic>,<at 0.5861 0.3251 RCIM_Cubic>]}
         }}))
     )
     "Float (Alpha)"=&Multiply_Float(
      A=$System.AlphaFade
      B=&FloatFromCurve(
       FloatCurve=DataInterface() {propertyValues {Curve {keys [
           {interpMode=RCIM_Cubic, leaveTangent=-1.8267, leaveTangentWeight=0.5724, tangentMode=RCTM_Break, time=0, value=1}
           {arriveTangent=-0.2139, interpMode=RCIM_Cubic, tangentMode=RCTM_Break, time=0.9, value=0}
          ]}}}
       Scale Curve=&FloatFromCurve(
        FloatCurve=DataInterface() {propertyValues {Curve {keys [
            <at 0.1 0.8 RCIM_Cubic>
            {arriveTangentWeight=0.0858, interpMode=RCIM_Cubic, leaveTangentWeight=0.0558, tangentMode=RCTM_User, time=0.4434, value=1}
            {arriveTangentWeight=0.0572, interpMode=RCIM_Cubic, leaveTangentWeight=0.1462, tangentMode=RCTM_User, time=0.6, value=1}
            <at 0.9 0.8 RCIM_Cubic>
           ]}}}
        CurveIndex=Linked() {linkedVariable {name=NPC.FortniteNPC.TimeOfDay_0-1, type=NiagaraFloat}}
        Scale Curve=&FN_FrontEndFloatModifier(Gameplay=0.25, Front End=0.15)
       )
      )
     )
    )) {isSetParameters=true}
   /SetVariables(Particles.TopLineColor:LinearColor=&MakeLinearColorFromVectorAndFloat(
     "Vector (RGB)"=&Multiply_Vector(
      A=&MakeVectorFromLinearColorRGB(Color=$System.TOD Color)
      B=&MergeHelper_VectorSwitch(
       0=&MakeVectorFromLinearColorRGB(Color=&ColorFromCurve(ColorCurve=DataInterface() {propertyValues {
           AlphaCurve {keys [<at 0 1 RCIM_Cubic>]}
           BlueCurve {keys [<at 0 0.8984 RCIM_Cubic>,<at 0.7117 0.2969 RCIM_Cubic>]}
           GreenCurve {keys [<at 0 0.6612 RCIM_Cubic>,<at 0.7117 0.1501 RCIM_Cubic>]}
           RedCurve {keys [<at 0 0.6825 RCIM_Cubic>,<at 0.7117 0.1469 RCIM_Cubic>]}
          }}))
       1=&MakeVectorFromLinearColorRGB(Color=&ColorFromCurve(ColorCurve=DataInterface() {propertyValues {
           AlphaCurve {keys [<at 0 1 RCIM_Cubic>]}
           BlueCurve {keys [<at 0 1 RCIM_Cubic>,<at 0.5861 0.5664 RCIM_Cubic>]}
           GreenCurve {keys [<at 0 0.4964 RCIM_Cubic>,<at 0.5861 0.2146 RCIM_Cubic>]}
           RedCurve {keys [<at 0 0.2455 RCIM_Cubic>,<at 0.5861 0.3251 RCIM_Cubic>]}
          }}))
       2=?
       3=?
       Index=$System.VariantInt
       Num=::NewEnumerator0
      )
     )
     "Float (Alpha)"=&Multiply_Float(
      A=$System.AlphaFade
      B=&FloatFromCurve(
       FloatCurve=DataInterface() {propertyValues {Curve {keys [
           {interpMode=RCIM_Cubic, leaveTangent=-1.8267, leaveTangentWeight=0.5724, tangentMode=RCTM_Break, time=0, value=1}
           {arriveTangent=-0.2139, interpMode=RCIM_Cubic, tangentMode=RCTM_Break, time=0.9, value=0}
          ]}}}
       Scale Curve=&FloatFromCurve(
        FloatCurve=DataInterface() {propertyValues {Curve {keys [
            <at 0.1 0.8 RCIM_Cubic>
            {arriveTangentWeight=0.0858, interpMode=RCIM_Cubic, leaveTangentWeight=0.0558, tangentMode=RCTM_User, time=0.4434, value=1}
            {arriveTangentWeight=0.0572, interpMode=RCIM_Cubic, leaveTangentWeight=0.1462, tangentMode=RCTM_User, time=0.6, value=1}
            <at 0.9 0.8 RCIM_Cubic>
           ]}}}
        CurveIndex=Linked() {linkedVariable {name=NPC.FortniteNPC.TimeOfDay_0-1, type=NiagaraFloat}}
        Scale Curve=&FN_FrontEndFloatModifier(Gameplay=0.25, Front End=0.15)
       )
      )
     )
    )) {isSetParameters=true}
   &ScaleRibbonWidth(Ribbon Width Scale=&FloatFromCurve(FloatCurve=DataInterface() {propertyValues {Curve {
        defaultValue=1
        keys [
         {interpMode=RCIM_Cubic, tangentMode=RCTM_User, time=0, value=1}
         {arriveTangent=-3.0576, interpMode=RCIM_Cubic, leaveTangent=-3.0576, tangentMode=RCTM_User, time=1, value=0}
        ]
       }}})) {enabled=false}
  ]
  rendererClasses [&NiagaraRibbonRendererProperties]
  renderers [{rendererClass=&NiagaraRibbonRendererProperties}]
 }
}
summary {
 systemName=NS_Pickaxe_Guitar_AnvilAlarmMoth_SwingTrail
 emitters [
  {emitterName=Trail, rendererClasses [&NiagaraRibbonRendererProperties]}
  {emitterName=NE_Main_Trail, rendererClasses [&NiagaraRibbonRendererProperties]}
  {emitterName=Accent_trail, rendererClasses [&NiagaraRibbonRendererProperties]}
  {emitterName=Accent_trail_Variant, rendererClasses [&NiagaraRibbonRendererProperties]}
 ]
 userVariables [
  User.bIsFrontEnd {type=NiagaraBool}
  User.bIsFrontEndPreview {type=NiagaraBool}
  User.RibbonIDOffset {type=NiagaraInt32}
  User.SkeletalMesh {
   type=NiagaraDataInterfaceSkeletalMesh
   defaultValue {
    struct=NiagaraExt_VariableValue_DataInterface
    value {dataInterface=&self:NiagaraDataInterfaceSkeletalMesh_0}
   }
  }
  User.SwingSectionIndex {type=NiagaraInt32}
  User.VariantID {type=NiagaraFloat, defaultValue=0, description=""}
 ]
}
